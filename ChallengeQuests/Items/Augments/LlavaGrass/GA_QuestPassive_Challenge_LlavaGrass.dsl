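; Passive challenge-quest ability for LlavaGrass that damages the player while on foot on listed ground surfaces, presumably grass.
; A surface-change gameplay event and a looping half-second timer run the same duplicated check.
; It excludes vehicles and tagged movement states, reads the surface tracking component, and applies the damage effect unless its tags are already active.
; The effect is removed by tag when the player is safe or the ability ends, which also clears the timer.
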(event Ability|EventFailedToActivatePassiveAbility
  (params (ActorInfo :type Gameplay%20Ability%20Actor%20Info%20Structure))
  (|CallParentFunction ActorInfo))

(event Ability|EventActivateAbility
  (|CallParentFunction)
  (bind _returnvalue (Utilities|Time|SetTimerbyEvent (delegate Custom|GroundCheck) 0.5 true true))
  (Variables|Default|SetTimer%20Handle _returnvalue)
  (Ability|Tasks|WaitGameplayEvent "(TagName=\"Event.SurfaceType.Changed\")" (Variables|GAVars|Data|GetOwningPlayerPawn)
    (:EventReceived
      (if (CallFunction|IsonFoot)
        (if (Default|IsTouchingGround)
          (if (not (GameplayTags|HasAnyMatchingGameplayTags (Ability|GetAbilitySystemComponent (Variables|GAVars|Data|GetOwningPlayerPawn)) (Variables|Default|GetEffectIsActiveTags)))
            ;[0.2083 0.9706 1] Player is on foot AND touching the ground, apply damage
            (GameplayEffects|ApplyGameplayEffectToSelf :self (Ability|GetAbilitySystemComponent (Variables|GAVars|Data|GetOwningPlayerPawn)) :GameplayEffectClass (Variables|Default|GetGE_GrassLlavaDMG) :Level 0.0))
          (else
            _))
        (else
          _))
      ;[0.2083 0.9706 1] Player is Safe, remove the effect
      (GameplayEffects|RemoveActiveEffectswithTags (Ability|GetAbilitySystemComponent (Variables|GAVars|Data|GetOwningPlayerPawn)) (Variables|Default|GetEffectIsActiveTags)))))

(event Custom|GroundCheck
  (if (CallFunction|IsonFoot)
    (if (Default|IsTouchingGround)
      (if (not (GameplayTags|HasAnyMatchingGameplayTags (Ability|GetAbilitySystemComponent (Variables|GAVars|Data|GetOwningPlayerPawn)) (Variables|Default|GetEffectIsActiveTags)))
        ;[0.2083 0.9706 1] Player is on foot AND touching the ground, apply damage
        (GameplayEffects|ApplyGameplayEffectToSelf :self (Ability|GetAbilitySystemComponent (Variables|GAVars|Data|GetOwningPlayerPawn)) :GameplayEffectClass (Variables|Default|GetGE_GrassLlavaDMG) :Level 0.0))
      (else
        _))
    (else
      _))
  ;[0.2083 0.9706 1] Player is Safe, remove the effect
  (GameplayEffects|RemoveActiveEffectswithTags (Ability|GetAbilitySystemComponent (Variables|GAVars|Data|GetOwningPlayerPawn)) (Variables|Default|GetEffectIsActiveTags)))

(event Ability|EventOnEndAbility
  (params (bWasCancelled :type Boolean))
  (|CallParentFunction bWasCancelled)
  (GameplayEffects|RemoveActiveEffectswithTags (Ability|GetAbilitySystemComponent (Variables|GAVars|Data|GetOwningPlayerPawn)) (Variables|Default|GetEffectIsActiveTags))
  (Utilities|Time|ClearandInvalidateTimerbyHandle (Variables|Default|GetTimer%20Handle)))

(fn IsOnFoot
  (return (not (or (GameplayTags|HasAnyMatchingGameplayTags (Variables|GAVars|Data|GetOwningPlayerPawn) (Variables|Default|GetT_MovementTags)) (Vehicle|IsinVehicle (Variables|GAVars|Data|GetOwningPlayerPawn))))))

(fn IsTouchingGround
  (return (Utilities|Array|ContainsItem (Variables|Default|GetListOfSurfaces) (Class|FortSurfaceTrackingComponent|GetCurrentSurfaceType (Footsteps|GetSurfaceTrackingComponent (Variables|GAVars|Data|GetOwningPlayerPawn))))))
